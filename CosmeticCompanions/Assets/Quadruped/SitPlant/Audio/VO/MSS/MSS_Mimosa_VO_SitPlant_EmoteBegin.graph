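# Mono one-shot MetaSound Source for the Mimosa SitPlant companion's vocal reaction when an emote begins.
# It wraps the shared MSS_Mimosa_MoodReact graph, feeding it six EmoteBegin wave variations plus neutral pitch and volume inputs, and passes through its audio and finished trigger.
# Companion VO attenuation and the Mimosa property sheet apply, with concurrency rules including a per-owner, single-voice override that prevents new plays.
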MetaSoundSource /CosmeticCompanions/Assets/Quadruped/SitPlant/Audio/VO/MSS/MSS_Mimosa_VO_SitPlant_EmoteBegin
inputs [
 PitchModulationRange.Max {type=Float}
 PitchModulationRange.Min {type=Float}
 TemplatePitchMultiplier {type=Float}
 TemplateVolumeMultiplier {type=Float}
 UE.Source.OnPlay {type=Trigger}
 VolumeModulationRange.Max {type=Float}
 VolumeModulationRange.Min {type=Float}
 Waves {type=WaveAsset:Array}
]
interfaces [UE.OutputFormat.Mono:1.0, UE.Source.OneShot:1.0, UE.Source:1.1]
outputs [
 "UE.OutputFormat.Mono.Audio:0" {type=Audio}
 UE.Source.OneShot.OnFinished {type=Trigger}
]
pages [Default {
  input_defaults {
   PitchModulationRange.Max=1
   PitchModulationRange.Min=1
   TemplatePitchMultiplier=1
   TemplateVolumeMultiplier=1
   UE.Source.OnPlay=null
   VolumeModulationRange.Max=1
   VolumeModulationRange.Min=1
   Waves [
    /CosmeticCompanions/Assets/Quadruped/SitPlant/Audio/VO/Waves/Mimosa_VO_SitPlant_Reaction_EmoteBegin_010
    /CosmeticCompanions/Assets/Quadruped/SitPlant/Audio/VO/Waves/Mimosa_VO_SitPlant_Reaction_EmoteBegin_012
    /CosmeticCompanions/Assets/Quadruped/SitPlant/Audio/VO/Waves/Mimosa_VO_SitPlant_Reaction_EmoteBegin_013
    /CosmeticCompanions/Assets/Quadruped/SitPlant/Audio/VO/Waves/Mimosa_VO_SitPlant_Reaction_EmoteBegin_014
    /CosmeticCompanions/Assets/Quadruped/SitPlant/Audio/VO/Waves/Mimosa_VO_SitPlant_Reaction_EmoteBegin_015
    /CosmeticCompanions/Assets/Quadruped/SitPlant/Audio/VO/Waves/Mimosa_VO_SitPlant_Reaction_EmoteBegin_016
   ]
  }
  nodes [MSS_Mimosa_MoodReact_0 = /FortMetasound/Cosmetics/Mimosa/Mood/MSS_Mimosa_MoodReact(
    PitchModulationRange.Max=^PitchModulationRange.Max
    PitchModulationRange.Min=^PitchModulationRange.Min
    TemplatePitchMultiplier=^TemplatePitchMultiplier
    TemplateVolumeMultiplier=^TemplateVolumeMultiplier
    UE.Source.OnPlay=^UE.Source.OnPlay
    VolumeModulationRange.Max=^VolumeModulationRange.Max
    VolumeModulationRange.Min=^VolumeModulationRange.Min
    Mood.Sounds=^Waves
   )]
  links [
   {dst="@UE.OutputFormat.Mono.Audio:0", src=MSS_Mimosa_MoodReact_0.UE.OutputFormat.Mono.Audio:0}
   {dst="@UE.Source.OneShot.OnFinished", src=MSS_Mimosa_MoodReact_0.UE.Source.OneShot.OnFinished}
  ]
 }]
properties {
 assetImportData=&self:AssetImportData
 attenuationSettings=/Game/Sounds/Attenuation/AS_Companion_VO
 audioPropertiesSheet=/CosmeticCompanions/Assets/Audio_Shared/PropertySheets/APS_Cosmetic_Mimosa_VO
 maxDistance=2200
 soundClassObject=&Fort_Creatures_Pets
 assetUserData [&self:AssetUserData_0]
 concurrencyOverrides {bLimitToOwner=true, maxCount=1, resolutionRule=PreventNew}
 concurrencySet [/Game/Sounds/Concurrency_Rules/CS_Mimosa_VO, /Game/Sounds/Concurrency_Rules/CS_Mimosa_VO_Global]
}
